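# Physics asset for the Valor pickaxe melee weapon's mesh.
# A large capsule body covers the two-handed handle, and a small capsule covers the dyn_Strap bone.
# One constraint links the strap to the handle with zero twist and first swing limits and a wide second swing limit, confining the strap's motion to a single swing axis.
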PhysicsAsset /BRCosmetics/Weapons/FORT_Melee/Pickaxe_Valor/Meshes/Valor_Pickaxe_Physics
bodies {
 Melee_TwoHanded_Handle [{shapeType=Capsule, length=60, radius=33.75, center=<0 0 42.6833>}]
 dyn_Strap [{shapeType=Capsule, length=12, radius=7.75, center=<0 0 -9.1276>}]
}
constraints [{bone1Name=dyn_Strap, bone2Name=Melee_TwoHanded_Handle, swing1LimitDegrees=0, swing2LimitDegrees=90, twistLimitDegrees=0}]
